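# Niagara system for the HighTower_Squash emote, staging a lightning-storm effect with body arcs, glowing eyes, lightning strikes, a dark ground vortex and orbiting debris.
# Arc and glow emitters follow User skeletal meshes through BoneFlow modules or bone-sampled locations, with offsets for the eyes.
# Delayed bursts time strikes and decals, some adding particle lights.
# Curves on System.NormalizedLoopAge cut spawning, fade glows and drop debris near the loop's end.
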NiagaraSystem /BRCosmetics/Effects/Fort_Effects/Emotes/HighTower_Squash/NS_Emote_HighTower_Squash
dependencies {
 usedDataInterfaces [&NiagaraDataInterfaceCurve, &NiagaraDataInterfaceColorCurve, &NiagaraDataInterfaceVectorCurve]
 usedDynamicInputs []
 usedModules [&SystemState]
}
emitters {
 Air_Swooping_Up_Mesh {
  bEnabled=false
  emitterUpdateScript [
   &EmitterState(Delay First Loop Only=true, Life Cycle Mode=::NewEnumerator1, Loop Delay=1, UseLoopDelay=true)
   &SpawnRate(SpawnRate=30, Use Spawn Probability=false)
  ]
  particleSpawnScript [
   /Niagara/Modules/Spawn/Initialization/InitializeParticle(
    Color=<rgba 0.0278 0.0583 0.0833 0.35>
    Lifetime=&UniformRangedFloat(Minimum=0.15, Maximum=0.2)
    Mesh Scale=&UniformRangedVector(Minimum=<0.75 0.75 1>, Maximum=<1 1 1.25>)
    Module.Write Color=true
    Module.Write Lifetime=true
    Module.Write Mass=true
    Module.Write Material Random=true
    Module.Write Position=false
    Module.Write Scale=true
    Module.Write SpriteRotation=false
    Module.Write SpriteSize=false
    Position=&SimulationPosition()
   )
   /Niagara/Modules/Spawn/Orientation/V2/InitialMeshOrientation(
    Rotation=&UniformRangedVector(Minimum=<0 0 0>, Maximum=<0 0 1>)
    Use Orientation Vector=true
    Use Rotation Vector=true
   )
   &SystemLocation(MaskSpawnGroup=false, Offset=<0 -10 0>, Offset Coordinate Space=::Simulation)
   &DynamicMaterialParameters(
    Index 0 Param 1=1
    Index 0 Param 2=&UniformRangedFloat(Minimum=0.5, Maximum=1)
    Index 0 Param 3=3
    Index 0 Param 4=1
    Module.Param0WriteEnabled=true
    Module.Param1WriteEnabled=true
    Module.Param2WriteEnabled=true
    Module.Param3WriteEnabled=true
   )
  ]
  particleUpdateScript [
   &ParticleState()
   &ScaleMeshSize(Scale Factor=&VectorFromCurve(VectorCurve=DataInterface() {propertyValues {XCurve {keys [<at 0 1>,<at 1 1.5>]}, YCurve {keys [<at 0 1.5>,<at 1 0.5>]}, ZCurve {keys [<at 0 0>,<at 1 2>]}}}))
   &ScaleColor(
    Scale Alpha=&Multiply_Float(
     A=&FloatFromCurve(FloatCurve=DataInterface() {propertyValues {Curve {keys [<at 0 0 RCIM_Cubic>,<at 0.5 1 RCIM_Cubic>,<at 1 0>]}}})
     B=&FloatFromCurve(
      FloatCurve=DataInterface() {propertyValues {Curve {keys [<at 0.55 1 RCIM_Cubic>, {arriveTangent=-4.5166, interpMode=RCIM_Cubic, leaveTangent=-4.5166, tangentMode=RCTM_User, time=0.7, value=0}]}}}
      CurveIndex=$System.NormalizedLoopAge
     )
    )
    ScaleA=true
    ScaleRGB=true
   )
  ]
  rendererClasses [&NiagaraMeshRendererProperties]
  renderers [{rendererClass=&NiagaraMeshRendererProperties, rendererIndex=0}]
 }
 BoneFlowElectricArcs {
  emitterUpdateScript [
   &EmitterState(Inactive Response=::NewEnumerator1, Life Cycle Mode=::NewEnumerator1, Loop Behavior=::NewEnumerator1, Loop Duration=4.25, UseLoopDelay=false)
   /Game/Effects/Niagara/Modules/Partition/FN_PartitionAttributes/FN_PartitionAttributes_v2(Particles Per Partition=5, Partitions=2)
   &SpawnRate(
    Spawn Probability=&FloatFromCurve(
     FloatCurve=DataInterface() {propertyValues {Curve {keys [<at 0.74 1 RCIM_Constant>, {arriveTangent=-4.5166, interpMode=RCIM_Constant, leaveTangent=-4.5166, time=0.75, value=0}]}}}
     CurveIndex=$System.NormalizedLoopAge
    )
    SpawnRate=&UniformRangedFloat(Minimum=24, Maximum=50)
    Use Spawn Probability=true
   )
   &SpawnBurst_Instantaneous(Spawn Count=14, Use Spawn Probability=false)
  ]
  particleSpawnScript [
   /Game/Effects/Niagara/Modules/Partition/FN_ParticlePartition_v2/FN_ParticlePartition_v2()
   &InitializeParticle(
    Color=&MakeLinearColorFromVectorAndFloat(
     "Vector (RGB)"=&MakeVectorFromLinearColorRGB(Color=<rgba 0.9405 1.1746 1.25 1>)
     "Float (Alpha)"=&UniformRangedFloat(Minimum=0.25, Maximum=1)
    )
    Lifetime=&UniformRangedFloat(Minimum=0.45, Maximum=0.65)
    Module.Write Color=true
    Module.Write Lifetime=true
    Module.Write Mass=false
    Module.Write Material Random=true
    Module.Write Position=false
    Module.Write Scale=false
    Module.Write SpriteRotation=true
    Module.Write SpriteSize=true
    Sprite Rotation=&NormalizedAngleToDegrees(NormalizedAngle=&UniformRangedFloat(Minimum=0, Maximum=1))
    Sprite Size=&Vector2DFromFloat(Value=&UniformRangedFloat(Minimum=8, Maximum=12))
   )
   /Game/Effects/Niagara/Modules/BoneFlow/FN_Initialize_BoneFlow_Attributes_v2(
    Angle:NiagaraFloat=&UniformRangedFloat(Minimum=0, Maximum=10)
    Radius=2
    T:NiagaraFloat=&UniformRangedFloat(Minimum=0, Maximum=1)
   )
   /SetVariables(Particles.SubImageIndex=&UniformRangedFloat(Minimum=0, Maximum=3)) {isSetParameters=true, module=SetVariables_#1}
   &DynamicMaterialParameters(
    Index 0 Param 1=&UniformRangedFloat(Minimum=0, Maximum=0.1)
    Index 0 Param 2=&UniformRangedFloat(Minimum=1, Maximum=2)
    Module.Param0WriteEnabled=true
    Module.Param1WriteEnabled=true
    Module.Param2WriteEnabled=false
    Module.Param3WriteEnabled=false
   )
   &CameraOffset(Camera Offset Amount=2)
  ]
  particleUpdateScript [
   &ParticleState()
   /Game/Effects/Niagara/Modules/BoneFlow/FN_BoneFlow_v2/FN_BoneFlow_v2(
    Angle:NiagaraFloat=$Particles.FN_Initialize_BoneFlow_Attributes_v2.Angle
    Radius:NiagaraFloat=$Particles.FN_Initialize_BoneFlow_Attributes_v2.Radius
    SkeletalMesh:NiagaraDataInterfaceSkeletalMesh=$User.SkeletalMesh
    T:NiagaraFloat=&Add_Float(A=$Particles.FN_Initialize_BoneFlow_Attributes_v2.T, B=&Multiply_Float(A=$Particles.NormalizedAge, B=0.1))
   )
   &DynamicMaterialParameters(
    Index 0 Param 3=&FloatFromCurve(FloatCurve=DataInterface() {propertyValues {Curve {keys [<at 0 1 RCIM_Cubic>,<at 0.2 0 RCIM_Cubic>,<at 1 1>]}}})
    Module.Param0WriteEnabled=false
    Module.Param1WriteEnabled=false
    Module.Param2WriteEnabled=true
    Module.Param3WriteEnabled=false
   ) {module=DynamicMaterialParameters001}
   &ScaleColor(
    Scale Alpha=&FloatFromCurve(
     FloatCurve=DataInterface() {propertyValues {Curve {keys [<at 0.65 1 RCIM_Cubic>, {arriveTangent=-4.5166, interpMode=RCIM_Cubic, leaveTangent=-4.5166, tangentMode=RCTM_User, time=1, value=0}]}}}
     CurveIndex=$System.NormalizedLoopAge
    )
    Scale RGB=&MakeVectorFromLinearColorRGB(Color=&ColorFromCurve(ColorCurve=DataInterface() {propertyValues {
        AlphaCurve {keys []}
        BlueCurve {keys [<at 0 3 RCIM_Cubic>,<at 1 1 RCIM_Cubic>]}
        GreenCurve {keys [<at 0 2.7135 RCIM_Cubic>,<at 1 0.8177 RCIM_Cubic>]}
        RedCurve {keys [<at 0 2.5625 RCIM_Cubic>,<at 1 0.8188 RCIM_Cubic>]}
       }}))
    ScaleA=true
    ScaleRGB=true
   )
  ]
  rendererClasses [&NiagaraSpriteRendererProperties]
  renderers [{rendererClass=&NiagaraSpriteRendererProperties}]
 }
 BoneFlowElectricArcsCenterBody {
  emitterUpdateScript [
   &EmitterState(Inactive Response=::NewEnumerator1, Life Cycle Mode=::NewEnumerator1, Loop Behavior=::NewEnumerator1, Loop Duration=4.25, UseLoopDelay=false)
   &FN_PartitionAttributes_v2(Particles Per Partition=30)
   &SpawnRate(
    Spawn Probability=&FloatFromCurve(
     FloatCurve=DataInterface() {propertyValues {Curve {keys [<at 0.74 1 RCIM_Constant>, {arriveTangent=-4.5166, interpMode=RCIM_Constant, leaveTangent=-4.5166, time=0.75, value=0}]}}}
     CurveIndex=$System.NormalizedLoopAge
    )
    SpawnRate=&UniformRangedFloat(Minimum=8, Maximum=12)
    Use Spawn Probability=true
   )
  ]
  particleSpawnScript [
   &FN_ParticlePartition_v2()
   &InitializeParticle(
    Color=&MakeLinearColorFromVectorAndFloat(
     "Vector (RGB)"=&MakeVectorFromLinearColorRGB(Color=<rgba 1.5048 1.8793 2 1>)
     "Float (Alpha)"=&UniformRangedFloat(Minimum=0.25, Maximum=1)
    )
    Lifetime=&UniformRangedFloat(Minimum=0.35, Maximum=0.65)
    Module.Write Color=true
    Module.Write Lifetime=true
    Module.Write Mass=false
    Module.Write Material Random=true
    Module.Write Position=false
    Module.Write Scale=false
    Module.Write SpriteRotation=true
    Module.Write SpriteSize=true
    Sprite Rotation=&NormalizedAngleToDegrees(NormalizedAngle=&UniformRangedFloat(Minimum=-1, Maximum=1))
    Sprite Size=&Vector2DFromFloat(Value=&UniformRangedFloat(Minimum=20, Maximum=24))
   )
   &FN_Initialize_BoneFlow_Attributes_v2(
    Angle:NiagaraFloat=&UniformRangedFloat(Minimum=0, Maximum=10)
    Radius=9
    T:NiagaraFloat=&UniformRangedFloat(Minimum=0, Maximum=1)
   )
   /SetVariables(Particles.SubImageIndex=&UniformRangedFloat(Minimum=0, Maximum=3)) {isSetParameters=true, module=SetVariables_#1}
   &DynamicMaterialParameters(
    Index 0 Param 1=&UniformRangedFloat(Minimum=0, Maximum=0.1)
    Index 0 Param 2=&UniformRangedFloat(Minimum=1, Maximum=2)
    Module.Param0WriteEnabled=true
    Module.Param1WriteEnabled=true
    Module.Param2WriteEnabled=false
    Module.Param3WriteEnabled=false
   )
   &CameraOffset(Camera Offset Amount=-1)
  ]
  particleUpdateScript [
   &ParticleState()
   &FN_BoneFlow_v2(
    Angle:NiagaraFloat=$Particles.FN_Initialize_BoneFlow_Attributes_v2.Angle
    BoneID:NiagaraInt32=&MakeIntFromFloat(Float=2)
    Radius:NiagaraFloat=&FloatFromCurve(FloatCurve=DataInterface() {propertyValues {Curve {keys [<at 0 14>,<at 1 8>]}}})
    SkeletalMesh:NiagaraDataInterfaceSkeletalMesh=$User.SkeletalMesh
    T:NiagaraFloat=&Add_Float(A=$Particles.FN_Initialize_BoneFlow_Attributes_v2.T, B=&Multiply_Float(A=$Particles.NormalizedAge, B=0.05))
   )
   &DynamicMaterialParameters(
    Index 0 Param 3=&FloatFromCurve(FloatCurve=DataInterface() {propertyValues {Curve {keys [<at 0 1 RCIM_Cubic>,<at 0.2 0 RCIM_Cubic>,<at 1 1>]}}})
    Module.Param0WriteEnabled=false
    Module.Param1WriteEnabled=false
    Module.Param2WriteEnabled=true
    Module.Param3WriteEnabled=false
   ) {module=DynamicMaterialParameters001}
   &ScaleColor(
    Scale Alpha=&FloatFromCurve(FloatCurve=DataInterface() {propertyValues {Curve {keys [<at 0.75 1 RCIM_Cubic>,<at 1 0>]}}})
    Scale RGB=&MakeVectorFromLinearColorRGB(Color=&ColorFromCurve(ColorCurve=DataInterface() {propertyValues {
        AlphaCurve {keys []}
        BlueCurve {keys [<at 0 3 RCIM_Cubic>,<at 1 1 RCIM_Cubic>]}
        GreenCurve {keys [<at 0 2.7135 RCIM_Cubic>,<at 1 0.8177 RCIM_Cubic>]}
        RedCurve {keys [<at 0 2.5625 RCIM_Cubic>,<at 1 0.8188 RCIM_Cubic>]}
       }}))
    ScaleA=true
    ScaleRGB=true
   )
  ]
  rendererClasses [&NiagaraSpriteRendererProperties]
  renderers [{rendererClass=&NiagaraSpriteRendererProperties}]
 }
 Dark_Spot_and_Upward_Vortex {
  emitterUpdateScript [
   &EmitterState(Life Cycle Mode=::NewEnumerator1, Loop Behavior=::NewEnumerator1, Loop Delay=0.75, Loop Duration=1, UseLoopDelay=true)
   &SpawnBurst_Instantaneous(Spawn Count=1, Use Spawn Probability=false)
  ]
  particleSpawnScript [
   &InitializeParticle(
    Color=<rgba 0.0399 0.0451 0.0521 10>
    Lifetime=3.25
    Mesh Scale=<0.4 0.4 0.35>
    Module.Write Color=true
    Module.Write Lifetime=true
    Module.Write Mass=true
    Module.Write Material Random=true
    Module.Write Position=true
    Module.Write Scale=true
    Module.Write SpriteRotation=false
    Module.Write SpriteSize=true
    Position=&SimulationPosition()
    Sprite Size=<300 300>
   )
   &CameraOffset(Camera Offset Amount=5)
   /Game/Effects/Niagara/Modules/Mesh/FN_MeshRenderer_v3/FN_MeshRenderer_v2(
    Color:LinearColor=<rgba 0.1788 0.2162 0.3698 1>
    Create COlor Attribute=true
    Create Dynamic Param Attribute=true
    "Create Normalized Age "=true
    DynamicMaterial:Vector4f=&MakeVector4(
     X=&UniformRangedFloat(Minimum=0, Maximum=1)
     Y=&UniformRangedFloat(Minimum=5, Maximum=6)
     Z=0
     W=2
    )
   )
   /NiagaraFN/Sprite/FN_SpriteAlignmentAndFacing(SpriteAlignment:Vector3f=<0 0 1>, SpriteFacing:Vector3f=<0 0 1>) {module=SpriteAlignmentAndFacing}
  ]
  particleUpdateScript [
   &ParticleState()
   &ScaleColor(
    Scale Alpha=&FloatFromCurve(FloatCurve=DataInterface() {propertyValues {Curve {keys [
         <at 0 0 RCIM_Cubic>
         {arriveTangent=1.1765, interpMode=RCIM_Cubic, leaveTangent=1.1765, time=0.002, value=0.008}
         <at 0.15 1>
         {arriveTangent=-0.008, interpMode=RCIM_Cubic, leaveTangent=-0.008, time=0.85, value=1}
         <at 1 0>
        ]}}})
    Scale RGB=&MakeVectorFromLinearColorRGB(Color=<rgba 1 1 1 1>)
    ScaleA=true
    ScaleRGB=true
   )
   /SetVariables(Particles.FN_MeshRenderer_v2.Color:LinearColor=&MakeLinearColorFromVectorAndFloat(
     "Vector (RGB)"=&MakeVectorFromLinearColorRGB(Color=<rgba 0.2338 0.3046 0.4219 1>)
     "Float (Alpha)"=&Multiply_Float(
      A=1
      B=&FloatFromCurve(FloatCurve=DataInterface() {propertyValues {Curve {keys [
           <at 0 0 RCIM_Cubic>
           {arriveTangent=1.3333, leaveTangent=1.3333, time=0.25, value=1}
           <at 0.75 1 RCIM_Cubic>
           {arriveTangent=-4.5166, interpMode=RCIM_Cubic, leaveTangent=-4.5166, tangentMode=RCTM_User, time=1, value=0}
          ]}}})
     )
    )) {isSetParameters=true}
   &ScaleSpriteSize()
   /Niagara/Modules/Update/Orientation/V2/UpdateMeshOrientation(Rotation Vector:Vector3f=<0 0 1>)
  ]
  rendererClasses [&NiagaraSpriteRendererProperties, &NiagaraMeshRendererProperties]
  renderers [
   {rendererClass=&NiagaraSpriteRendererProperties}
   {rendererClass=&NiagaraMeshRendererProperties, rendererIndex=1}
  ]
 }
 Debris_Vortex {
  emitterUpdateScript [
   &EmitterState(Life Cycle Mode=::NewEnumerator1, Loop Delay=0.5, Loop Duration=1, UseLoopDelay=true)
   &SpawnRate(
    Spawn Probability=&FloatFromCurve(
     FloatCurve=DataInterface() {propertyValues {Curve {keys [<at 0.74 1 RCIM_Constant>, {arriveTangent=-4.5166, interpMode=RCIM_Constant, leaveTangent=-4.5166, time=0.75, value=0}]}}}
     CurveIndex=$System.NormalizedLoopAge
    )
    SpawnRate=&UniformRangedFloat(Minimum=6, Maximum=20)
    Use Spawn Probability=true
   )
  ]
  particleSpawnScript [
   /Game/Effects/Niagara/Modules/Orbit/FN_Orbit_Init_Variables_v2()
   &InitializeParticle(
    Color=&UniformRangedLinearColor(Minimum=<rgba 0.1198 0.0673 0 1>, Maximum=<rgba 0.0677 0.052 0 1>)
    Lifetime=&UniformRangedFloat(Minimum=1, Maximum=2)
    Mass=0.25
    Module.Write Color=true
    Module.Write Lifetime=true
    Module.Write Mass=true
    Module.Write Material Random=true
    Module.Write Position=false
    Module.Write Scale=false
    Module.Write SpriteRotation=true
    Module.Write SpriteSize=true
    Position=&SimulationPosition()
    Sprite Rotation=&NormalizedAngleToDegrees(NormalizedAngle=&UniformRangedFloat(Minimum=-1, Maximum=1))
    Sprite Size=&Vector2DFromFloat(Value=&UniformRangedFloat(Minimum=7, Maximum=10))
   )
   /SetVariables(Particles.RockSubImageIndex:NiagaraFloat=&UniformRangedFloat(Minimum=0, Maximum=3)) {isSetParameters=true}
   &AddVelocity(Velocity=&UniformRangedVector(Minimum=<0 0 50>, Maximum=<0 0 100>))
  ]
  particleUpdateScript [
   &ParticleState()
   /Game/Effects/Niagara/Modules/Orbit/FN_Orbit_v3(Belly Down Orientation=true, Height min/max:Vector2f=<0 5>, Radius min/max:Vector2f=<50 100>, Rotation Rate min/max:Vector2f=<0.25 1.5>) {enabled=false}
   /Game/Effects/Niagara/Modules/Orbit/FN_Orbit_v4(Height min/max:Vector2f=<0 5>, Radius min/max:Vector2f=<50 100>, Rotation Rate min/max:Vector2f=<0.25 1.5>)
   &GravityForce(Gravity=&MakeVector(
     X=0
     Y=0
     Z=&FloatFromCurve(
      FloatCurve=DataInterface() {propertyValues {Curve {keys [<at 0.725 0 RCIM_Cubic>,<at 0.75 -2000 RCIM_Constant>]}}}
      CurveIndex=$System.NormalizedLoopAge
     )
    ))
   &SolveForcesAndVelocity(Clamp Velocity=false, Limit Acceleration=false)
   &SpriteRotationRate(Rotation Rate=360)
   &ScaleSpriteSize(Scale Factor=&Vector2DFromFloat(Value=&Multiply_Float(
      A=&FloatFromCurve(FloatCurve=DataInterface() {propertyValues {Curve {keys [<at 0 0 RCIM_Cubic>,<at 0.1 1 RCIM_Cubic>,<at 1 0>]}}})
      B=&FloatFromCurve(
       FloatCurve=DataInterface() {propertyValues {Curve {keys [<at 0.65 1 RCIM_Cubic>, {arriveTangent=-4.5166, interpMode=RCIM_Cubic, leaveTangent=-4.5166, tangentMode=RCTM_User, time=1, value=0}]}}}
       CurveIndex=$System.NormalizedLoopAge
      )
     )))
  ]
  rendererClasses [&NiagaraSpriteRendererProperties]
  renderers [{rendererClass=&NiagaraSpriteRendererProperties}]
 }
 Decal {
  emitterUpdateScript [
   &EmitterState(Life Cycle Mode=::NewEnumerator1, Loop Behavior=::NewEnumerator1, Loop Delay=0.4, Loop Duration=1, UseLoopDelay=true)
   &SpawnBurst_Instantaneous(Spawn Count=13, Use Spawn Probability=false)
  ]
  particleSpawnScript [
   &FN_SetSpawnGroup_v2(Number of Spawn Groups=13)
   &InitializeParticle(
    Color=<rgba 1.1523 1.2291 1.25 1>
    Lifetime=0.65
    Module.Write Color=true
    Module.Write Lifetime=true
    Module.Write Mass=true
    Module.Write Material Random=true
    Module.Write Position=false
    Module.Write Scale=false
    Module.Write SpriteRotation=false
    Module.Write SpriteSize=true
    Position=&SimulationPosition()
    Sprite Size=&Vector2DFromFloat(Value=35)
   )
   &DynamicMaterialParameters(
    Index 0 Param 1=0.5
    Index 0 Param 2=1
    Index 0 Param 3=1
    Module.Param0WriteEnabled=false
    Module.Param1WriteEnabled=true
    Module.Param2WriteEnabled=true
    Module.Param3WriteEnabled=true
   )
  ]
  particleUpdateScript [
   &ParticleState()
   &SkeletalMeshLocation(
    Apply Sampled Orientation to Mesh Particles=true
    Bone / Socket Index=$Particles.SpawnGroup
    Bone Sampling Mode=::NewEnumerator4
    Inherit Velocity=false
    MaskSpawnGroup=false
    Skeletal Mesh=$User.SkeletalMesh_03
   )
   &DynamicMaterialParameters(
    Index 0 Param 1=&FloatFromCurve(FloatCurve=DataInterface() {propertyValues {Curve {keys [<at 0.4 3>,<at 1 6>]}}})
    Module.Param0WriteEnabled=true
    Module.Param1WriteEnabled=false
    Module.Param2WriteEnabled=false
    Module.Param3WriteEnabled=false
   ) {module=DynamicMaterialParameters001}
   &ScaleColor(
    Scale Alpha=&FloatFromCurve(FloatCurve=DataInterface() {propertyValues {Curve {keys [<at 0.75 1>,<at 1 0>]}}})
    ScaleA=true
    ScaleRGB=true
   )
   &CameraOffset(Camera Offset Amount=20)
  ]
  rendererClasses [&NiagaraSpriteRendererProperties]
  renderers [{rendererClass=&NiagaraSpriteRendererProperties}]
 }
 Eye_Electric_Arcs {
  emitterUpdateScript [
   &EmitterState(Life Cycle Mode=::NewEnumerator1, Loop Duration=1, Recalculate Duration Each Loop=true, UseLoopDelay=false)
   &SpawnRate(
    Spawn Probability=&FloatFromCurve(
     FloatCurve=DataInterface() {propertyValues {Curve {keys [<at 0.74 1 RCIM_Constant>, {arriveTangent=-4.5166, interpMode=RCIM_Constant, leaveTangent=-4.5166, time=0.75, value=0}]}}}
     CurveIndex=$System.NormalizedLoopAge
    )
    SpawnRate=8
    Use Spawn Probability=true
   )
  ]
  particleSpawnScript [
   &InitializeParticle(
    Color=&MakeLinearColorFromVectorAndFloat(
     "Vector (RGB)"=&MakeVectorFromLinearColorRGB(Color=<rgba 0.9405 1.1746 1.25 1>)
     "Float (Alpha)"=&UniformRangedFloat(Minimum=0.75, Maximum=1)
    )
    Lifetime=&UniformRangedFloat(Minimum=0.75, Maximum=1)
    Module.Write Color=true
    Module.Write Lifetime=true
    Module.Write Mass=false
    Module.Write Material Random=true
    Module.Write Position=false
    Module.Write Scale=false
    Module.Write SpriteRotation=true
    Module.Write SpriteSize=true
    Sprite Rotation=&NormalizedAngleToDegrees(NormalizedAngle=&UniformRangedFloat(Minimum=-1, Maximum=1))
    Sprite Size=&Vector2DFromFloat(Value=&UniformRangedFloat(Minimum=3, Maximum=5))
   )
   /SetVariables(Particles.SubImageIndex=&UniformRangedFloat(Minimum=0, Maximum=3)) {isSetParameters=true, module=SetVariables_#1}
   &DynamicMaterialParameters(
    Index 0 Param 1=&UniformRangedFloat(Minimum=0, Maximum=0.15)
    Index 0 Param 2=&UniformRangedFloat(Minimum=1, Maximum=2)
    Module.Param0WriteEnabled=true
    Module.Param1WriteEnabled=true
    Module.Param2WriteEnabled=false
    Module.Param3WriteEnabled=false
   )
   &CameraOffset(Camera Offset Amount=1)
   /SetVariables(Particles.AorBFloat:NiagaraFloat=&UniformAOrBFloat(A=0)) {isSetParameters=true}
  ]
  particleUpdateScript [
   &ParticleState()
   &SkeletalMeshLocation(
    Apply Sampled Orientation to Mesh Particles=true
    Bone / Socket Index=12
    Bone Sampling Mode=::NewEnumerator4
    Inherit Velocity=false
    MaskSpawnGroup=false
    Skeletal Mesh=$User.SkeletalMesh
   )
   /SetVariables(Particles.EyeElectricArcsPosition:Vector3f=&Add_Vector(
     A=$Output.SkeletalMeshLocation.SampledPosition
     B=&RotateVectorByQuaternion(
      Vector To Rotate=&Lerp_Vector(A=<3.5 7 3>, B=<3.5 7 -3>, Alpha=$Particles.Initial.AorBFloat)
      Quaternion=$Output.SkeletalMeshLocation.SampledRotation
     )
    )) {isSetParameters=true}
   &DynamicMaterialParameters(
    Index 0 Param 3=&FloatFromCurve(FloatCurve=DataInterface() {propertyValues {Curve {keys [<at 0 1 RCIM_Cubic>,<at 0.2 0 RCIM_Cubic>,<at 1 1>]}}})
    Module.Param0WriteEnabled=false
    Module.Param1WriteEnabled=false
    Module.Param2WriteEnabled=true
    Module.Param3WriteEnabled=false
   ) {module=DynamicMaterialParameters001}
   &ScaleColor(
    Scale Alpha=&FloatFromCurve(FloatCurve=DataInterface() {propertyValues {Curve {keys [<at 0.5 1 RCIM_Cubic>,<at 1 0>]}}})
    Scale RGB=&Multiply_VectorByFloat(
     Vector=&MakeVectorFromLinearColorRGB(Color=&ColorFromCurve(ColorCurve=DataInterface() {propertyValues {
         AlphaCurve {keys []}
         BlueCurve {keys [<at 0 3 RCIM_Cubic>,<at 1 1 RCIM_Cubic>]}
         GreenCurve {keys [<at 0 2.7135 RCIM_Cubic>,<at 1 0.8177 RCIM_Cubic>]}
         RedCurve {keys [<at 0 2.5625 RCIM_Cubic>,<at 1 0.8188 RCIM_Cubic>]}
        }}))
     Float=1
    )
    ScaleA=true
    ScaleRGB=true
   )
  ]
  rendererClasses [&NiagaraSpriteRendererProperties]
  renderers [{rendererClass=&NiagaraSpriteRendererProperties}]
 }
 Eye_Glows_and_Trails {
  emitterUpdateScript [
   &EmitterState(Inactive Response=::NewEnumerator2, Life Cycle Mode=::NewEnumerator1, Loop Behavior=::NewEnumerator1, Loop Duration=1, UseLoopDelay=false)
   &SpawnBurst_Instantaneous(Spawn Count=1, Use Spawn Probability=false)
  ]
  particleSpawnScript [&InitializeParticle(
    Color=<rgba 3.3437 4.9866 6 1>
    Lifetime=&UniformRangedFloat(Minimum=0.75, Maximum=1.5)
    Mesh Scale=<1 0.9 0.9>
    Module.Write Color=true
    Module.Write Lifetime=true
    Module.Write Mass=true
    Module.Write Material Random=true
    Module.Write Position=false
    Module.Write Scale=true
    Module.Write SpriteRotation=false
    Module.Write SpriteSize=false
    Position=&SimulationPosition()
   )]
  particleUpdateScript [
   &ParticleState(Kill Particles When Lifetime Has Elapsed=false)
   &SkeletalMeshLocation(
    Apply Sampled Orientation to Mesh Particles=true
    Bone / Socket Index=12
    Bone Sampling Mode=::NewEnumerator4
    Inherit Velocity=false
    MaskSpawnGroup=false
    Skeletal Mesh=$User.SkeletalMesh
   )
   /SetVariables(Particles.EyeGlowOrientation:Quat4f=&MultiplyQuaternion(
     Quaternion A=$Particles.MeshOrientation
     Quaternion B=&FindQuatBetweenTwoVectors(From Vector=<0 0 1>, To Vector=<1 0 0>)
    )) {isSetParameters=true}
   /SetVariables(Particles.EyeGlowPosition:Vector3f=&Add_Vector(
     A=$Output.SkeletalMeshLocation.SampledPosition
     B=&RotateVectorByQuaternion(Vector To Rotate=<0.5 0.5 0>, Quaternion=$Output.SkeletalMeshLocation.SampledRotation)
    )) {isSetParameters=true}
   /SetVariables(Particles.Position=&Add_Vector(
     A=$Output.SkeletalMeshLocation.SampledPosition
     B=&RotateVectorByQuaternion(Vector To Rotate=<0.2 0.7 0>, Quaternion=$Output.SkeletalMeshLocation.SampledRotation)
    )) {isSetParameters=true}
   &ScaleColor(
    Scale Alpha=&FloatFromCurve(
     FloatCurve=DataInterface() {propertyValues {Curve {keys [
         <at 0 0 RCIM_Cubic>
         <at 0.1 1>
         {arriveTangent=-2.3853, interpMode=RCIM_Cubic, leaveTangent=-2.3853, tangentMode=RCTM_User, time=0.75, value=1}
         <at 0.8 0 RCIM_Cubic>
        ]}}}
     CurveIndex=$System.NormalizedLoopAge
    )
    ScaleA=true
    ScaleRGB=true
   )
  ]
  rendererClasses [&NiagaraMeshRendererProperties]
  renderers [
   {rendererClass=&NiagaraMeshRendererProperties, rendererIndex=0}
   {rendererClass=&NiagaraMeshRendererProperties, rendererIndex=1}
  ]
 }
 FN_BoneFlow_v2 {
  emitterUpdateScript [
   &EmitterState(Life Cycle Mode=::NewEnumerator1, Loop Behavior=::NewEnumerator1, Loop Delay=0.5, Loop Duration=0.2, UseLoopDelay=true)
   &FN_PartitionAttributes_v2(Partitions=2)
   &SpawnRate(SpawnRate=&UniformRangedFloat(Minimum=10, Maximum=20), Use Spawn Probability=false)
  ]
  particleSpawnScript [
   &FN_ParticlePartition_v2()
   &InitializeParticle(
    Color=<rgba 6 6 6 1>
    Module.Write Color=true
    Module.Write Lifetime=true
    Module.Write Mass=true
    Module.Write Material Random=true
    Module.Write Position=false
    Module.Write Scale=false
    Module.Write SpriteRotation=true
    Module.Write SpriteSize=true
    Sprite Rotation=&UniformRangedFloat(Minimum=0, Maximum=360)
    Sprite Size=&Vector2DFromFloat(Value=&UniformRangedFloat(Minimum=10, Maximum=15))
   )
   &DynamicMaterialParameters(
    Index 0 Param 3=&FloatFromCurve(FloatCurve=DataInterface() {propertyValues {Curve {keys [<at 0 1 RCIM_Cubic>,<at 0.2 0.025 RCIM_Cubic>,<at 1 1>]}}})
    Module.Param0WriteEnabled=false
    Module.Param1WriteEnabled=false
    Module.Param2WriteEnabled=true
    Module.Param3WriteEnabled=false
   )
   &FN_Initialize_BoneFlow_Attributes_v2(
    Angle:NiagaraFloat=&UniformRangedFloat(Minimum=0, Maximum=360)
    Radius:NiagaraFloat=&UniformRangedFloat(Minimum=3, Maximum=10)
   )
   /SetVariables(Particles.SubImageIndex=&UniformRangedFloat(Minimum=0, Maximum=3)) {isSetParameters=true}
  ]
  particleUpdateScript [
   &ParticleState()
   /Game/Effects/Niagara/Modules/BoneFlow/FN_BoneFlow_v3/FN_BoneFlow_v3(
    Angle:NiagaraFloat=$Particles.FN_Initialize_BoneFlow_Attributes_v2.Angle
    Radius:NiagaraFloat=$Particles.FN_Initialize_BoneFlow_Attributes_v2.Radius
    SkeletalMesh:NiagaraDataInterfaceSkeletalMesh=$User.SkeletalMesh2
    T:NiagaraFloat=&Add_Float(A=$Particles.T, B=$Particles.NormalizedAge)
   )
   &DynamicMaterialParameters(
    Index 0 Param 3=&FloatFromCurve(FloatCurve=DataInterface() {propertyValues {Curve {keys [<at 0 1 RCIM_Cubic>,<at 0.2 0.025 RCIM_Cubic>,<at 1 1>]}}})
    Module.Param0WriteEnabled=false
    Module.Param1WriteEnabled=false
    Module.Param2WriteEnabled=true
    Module.Param3WriteEnabled=false
   ) {module=DynamicMaterialParameters001}
   &ScaleColor(
    Scale Alpha=&FloatFromCurve(FloatCurve=DataInterface() {propertyValues {Curve {keys [<at 0.75 1 RCIM_Cubic>, {arriveTangent=-10.3917, interpMode=RCIM_Cubic, leaveTangent=-10.3917, tangentMode=RCTM_User, time=1, value=0}]}}})
    Scale RGB=&Multiply_VectorByFloat(
     Vector=&MakeVectorFromLinearColorRGB(Color=&ColorFromCurve(ColorCurve=DataInterface() {propertyValues {
         AlphaCurve {keys []}
         BlueCurve {keys [<at 0 1>,<at 0.15 4.5 RCIM_Cubic>,<at 1 1 RCIM_Cubic>]}
         GreenCurve {keys [<at 0 1>, {arriveTangent=-0.1823, interpMode=RCIM_Cubic, leaveTangent=-0.1823, time=0.15, value=4.0703}, <at 1 0.8177 RCIM_Cubic>]}
         RedCurve {keys [<at 0 1>, {arriveTangent=-0.1812, interpMode=RCIM_Cubic, leaveTangent=-0.1812, time=0.15, value=3.8438}, <at 1 0.8188 RCIM_Cubic>]}
        }}))
     Float=1
    )
    ScaleA=true
    ScaleRGB=true
   )
   &CameraOffset(Camera Offset Amount=&UniformRangedFloat(Minimum=-10, Maximum=10))
   /Niagara/Modules/Update/SubUV/SubUVAnimation(Number Of Frames=8, Random SubUV Lookup Index:NiagaraFloat=$Particles.NormalizedAge, SubUV Animation Mode=::NewEnumerator1)
   &FN_Light_v3(
    Color=<rgba 10 10 10 1>
    Create Color Attribute=true
    Create Exponent Attribute=true
    Create Radius Attribute=true
    Enable Particle Light Enable=true
    Radius=25
   )
  ]
  rendererClasses [&NiagaraSpriteRendererProperties, &NiagaraLightRendererProperties]
  renderers [
   {rendererClass=&NiagaraSpriteRendererProperties}
   {rendererClass=&NiagaraLightRendererProperties, rendererIndex=1}
  ]
 }
 FN_BoneFlow_v2001 {
  emitterUpdateScript [
   &EmitterState(Life Cycle Mode=::NewEnumerator1, Loop Behavior=::NewEnumerator1, Loop Delay=0.5, Loop Duration=0.25, UseLoopDelay=true)
   &FN_PartitionAttributes_v2(Partitions=2)
   &SpawnRate(SpawnRate=40, Use Spawn Probability=false)
  ]
  particleSpawnScript [
   &FN_ParticlePartition_v2()
   &InitializeParticle(
    Color=<rgba 6.9583 7.7462 8 0.5>
    Module.Write Color=true
    Module.Write Lifetime=true
    Module.Write Mass=false
    Module.Write Material Random=true
    Module.Write Position=false
    Module.Write Scale=false
    Module.Write SpriteRotation=true
    Module.Write SpriteSize=true
    Position=&SimulationPosition()
    Sprite Rotation=&NormalizedAngleToDegrees(NormalizedAngle=&UniformRangedFloat(Minimum=-1, Maximum=1))
    Sprite Size=&Vector2DFromFloat(Value=&UniformRangedFloat(Minimum=20, Maximum=30))
   )
   &DynamicMaterialParameters(
    Index 0 Param 2=&UniformRangedFloat(Minimum=0, Maximum=10)
    Index 0 Param 3=1
    Index 0 Param 4=1
    Module.Param0WriteEnabled=false
    Module.Param1WriteEnabled=true
    Module.Param2WriteEnabled=true
    Module.Param3WriteEnabled=true
   ) {module=DynamicMaterialParameters001}
   &FN_Initialize_BoneFlow_Attributes_v2(
    Angle:NiagaraFloat=&UniformRangedFloat(Minimum=0, Maximum=360)
    Radius:NiagaraFloat=&UniformRangedFloat(Minimum=3, Maximum=10)
   )
  ]
  particleUpdateScript [
   &ParticleState()
   &FN_BoneFlow_v2(
    Angle:NiagaraFloat=$Particles.FN_Initialize_BoneFlow_Attributes_v2.Angle
    Radius:NiagaraFloat=&Lerp_Float(A=$Particles.FN_Initialize_BoneFlow_Attributes_v2.Radius, B=3, Alpha=$Emitter.NormalizedLoopAge)
    SkeletalMesh:NiagaraDataInterfaceSkeletalMesh=$User.SkeletalMesh2
    T:NiagaraFloat=&Add_Float(A=$Particles.T, B=&Multiply_Float(A=$Particles.NormalizedAge, B=0.25))
   )
   &CameraOffset(Camera Offset Amount=&UniformRangedFloat(Minimum=-10, Maximum=10))
   &DynamicMaterialParameters(
    Index 0 Param 1=&FloatFromCurve(FloatCurve=DataInterface() {propertyValues {Curve {keys [<at 0 0.01>,<at 1 1.5>]}}})
    Module.Param0WriteEnabled=true
    Module.Param1WriteEnabled=false
    Module.Param2WriteEnabled=false
    Module.Param3WriteEnabled=false
   ) {module=DynamicMaterialParameters002}
   &ScaleColor(
    Scale Alpha=&FloatFromCurve(FloatCurve=DataInterface() {propertyValues {Curve {keys [<at 0 1>,<at 0.4243 0.0508>,<at 1 1>]}}})
    Scale RGB=&VectorFromFloat(Value=&FloatFromCurve(FloatCurve=DataInterface() {propertyValues {Curve {keys [<at 0 1>,<at 0.1 1.5>,<at 1 1>]}}}))
    ScaleA=true
    ScaleRGB=true
   )
   &ScaleSpriteSize(Scale Factor=&Vector2DFromFloat(Value=&FloatFromCurve(FloatCurve=DataInterface() {propertyValues {Curve {keys [<at 0 1>,<at 1 0.25>]}}})))
  ]
  rendererClasses [&NiagaraSpriteRendererProperties]
  renderers [{rendererClass=&NiagaraSpriteRendererProperties}]
 }
 Initial_Lightning_Strike_LT {
  emitterUpdateScript [
   &EmitterState(
    Life Cycle Mode=::NewEnumerator1
    Loop Behavior=::NewEnumerator1
    Loop Delay=&UniformRangedFloat(Minimum=0, Maximum=0.2)
    Loop Duration=1
    UseLoopDelay=true
   )
   &SpawnBurst_Instantaneous(Spawn Count=1, Spawn Time=0.35, Use Spawn Probability=false)
   &SpawnBurst_Instantaneous(Spawn Count=1, Spawn Time=0.55, Use Spawn Probability=false) {module=SpawnBurst_Instantaneous001}
  ]
  particleSpawnScript [
   &FN_SetSpawnGroup_v2(Number of Spawn Groups=1)
   /Game/Effects/Niagara/Modules/Math/FN_EvenNumberToSign(Input:NiagaraInt32=$Particles.SpawnGroup, Invert Sign=true, Output Float Attribute=true)
   &InitializeParticle(
    Lifetime=&UniformRangedFloat(Minimum=0.075, Maximum=0.1)
    Mesh Scale=&Multiply_Vector(
     A=&UniformRangedVector(Minimum=<1.25 0.8 0.8>, Maximum=<1.5 1.5 1.5>)
     B=&MakeVector(X=$Particles.Sign, Y=1, Z=1)
    )
    Module.Write Color=true
    Module.Write Lifetime=true
    Module.Write Mass=true
    Module.Write Material Random=true
    Module.Write Position=false
    Module.Write Scale=true
    Module.Write SpriteRotation=true
    Module.Write SpriteSize=true
    Position=&SimulationPosition()
    Sprite Rotation=&UniformRangedFloat(Minimum=0, Maximum=360)
    Sprite Size=&Vector2DFromFloat(Value=&UniformRangedFloat(Minimum=140, Maximum=150))
   )
   &DynamicMaterialParameters(
    Index 0 Param 4=&UniformRangedFloat(Minimum=0, Maximum=1)
    Index 2 Param 1=0.2
    Index 2 Param 2=1
    Module.Param0WriteEnabled=false
    Module.Param0WriteEnabled1=true
    Module.Param0WriteEnabled2=true
    Module.Param0WriteEnabled3=true
    Module.Param1WriteEnabled=false
    Module.Param1WriteEnabled1=true
    Module.Param1WriteEnabled2=true
    Module.Param1WriteEnabled3=true
    Module.Param2WriteEnabled=false
    Module.Param2WriteEnabled1=true
    Module.Param2WriteEnabled2=true
    Module.Param2WriteEnabled3=true
    Module.Param3WriteEnabled=true
    Module.Param3WriteEnabled1=true
    Module.Param3WriteEnabled2=true
    Module.Param3WriteEnabled3=true
    Write Parameter Index 1=true
    Write Parameter Index 2=true
    Write Parameter Index 3=true
   )
   &InitialMeshOrientation(
    Rotation=&UniformRangedVector(Minimum=<-0.5 0 0>, Maximum=<0.5 0 0>)
    Use Orientation Vector=true
    Use Rotation Vector=true
   )
   /SetVariables(Particles.FlashCameraOffset=10) {isSetParameters=true, module=SetVariables_#2}
   /SetVariables(
    Particles.FlareOffset:Vector3f=&UniformRangedVector(Minimum=<5 -1 -1>, Maximum=<100 1 1>)
    Particles.FlareSize:Vector2f=&Vector2DFromFloat(Value=&UniformRangedFloat(Minimum=55, Maximum=60))
    Particles.LensFlare_Color:LinearColor=&MultiplyLinearColor(
     Linear Color=$Particles.Color
     Scale Factor=&MakeVector4_FromLinearColor(LinearColor=$System.RandomFlareColor)
    )
   ) {isSetParameters=true, module=SetVariables_#3}
  ]
  particleUpdateScript [
   &SkeletalMeshLocation(
    Apply Sampled Orientation to Mesh Particles=true
    Bone / Socket Index=&Add_Integer(A=&SpawnGroup(), B=9)
    Bone Sampling Mode=::NewEnumerator4
    Inherit Velocity=false
    MaskSpawnGroup=false
    Skeletal Mesh=$User.SkeletalMesh
   )
   /SetVariables(Particles.MeshOrientation=&MultiplyQuaternion(Quaternion A=$Output.SkeletalMeshLocation.SampledRotation, Quaternion B=$Particles.Initial.MeshOrientation)) {isSetParameters=true, module=SetVariables_#4}
   /SetVariables(Particles.FlashPosition:Vector3f=&Add_Vector(A=Linked() {linkedVariable {name=Particles.Position, type=NiagaraPosition}}, B=<0 0 10>)) {isSetParameters=true, module=SetVariables_#5}
   /SetVariables(Particles.FlashColor:LinearColor=&ColorFromCurve(ColorCurve=DataInterface() {propertyValues {
       AlphaCurve {keys [<at 0.3517 0>,<at 0.3988 1>,<at 0.9585 1 RCIM_Cubic>,<at 1 0 RCIM_Cubic>]}
       BlueCurve {keys [<at 0.664 4 RCIM_Cubic>,<at 0.946 2 RCIM_Cubic>]}
       GreenCurve {keys [<at 0.664 3.6548 RCIM_Cubic>,<at 0.946 1.8793 RCIM_Cubic>]}
       RedCurve {keys [<at 0.664 2.5833 RCIM_Cubic>,<at 0.946 1.5048 RCIM_Cubic>]}
      }})) {isSetParameters=true, module=SetVariables_#6}
   &ParticleState()
   &ScaleColor(
    Scale RGB=&MakeVectorFromLinearColorRGB(Color=&ColorFromCurve(ColorCurve=DataInterface() {propertyValues {
        AlphaCurve {keys []}
        BlueCurve {keys [<at 0 2 RCIM_Cubic>,<at 1 1 RCIM_Cubic>]}
        GreenCurve {keys [<at 0 1.9431 RCIM_Cubic>,<at 1 0.8596 RCIM_Cubic>]}
        RedCurve {keys [<at 0 1.8021 RCIM_Cubic>,<at 1 0.6667 RCIM_Cubic>]}
       }}))
    ScaleA=true
    ScaleRGB=true
   )
   &FN_Light_v3(
    Color=&ColorFromCurve(ColorCurve=DataInterface() {propertyValues {
       AlphaCurve {keys []}
       BlueCurve {keys [
         <at 0.3882 4 RCIM_Cubic>
         <at 0.5 25 RCIM_Cubic>
         {arriveTangent=-18.6776, interpMode=RCIM_Cubic, leaveTangent=-18.6776, time=0.685, value=4}
         <at 0.8 19.3967 RCIM_Cubic>
         <at 1 4 RCIM_Cubic>
        ]}
       GreenCurve {keys [
         <at 0.3882 3.6948 RCIM_Cubic>
         <at 0.5 22.345 RCIM_Cubic>
         {arriveTangent=-12.3711, interpMode=RCIM_Cubic, leaveTangent=-12.3711, time=0.685, value=3.6948}
         <at 0.8 18.6337 RCIM_Cubic>
         <at 1 3.6948 RCIM_Cubic>
        ]}
       RedCurve {keys [
         <at 0.3882 3.2 RCIM_Cubic>
         <at 0.5 18.041 RCIM_Cubic>
         {arriveTangent=-2.1474, interpMode=RCIM_Cubic, leaveTangent=-2.1474, time=0.685, value=3.2}
         <at 0.8 17.3967 RCIM_Cubic>
         <at 1 3.2 RCIM_Cubic>
        ]}
      }})
    Create Color Attribute=true
    Create Exponent Attribute=true
    Create Position Attribute=true
    Create Radius Attribute=true
    Exponent=1.5
    Position:Vector3f=&Add_Vector(A=Linked() {linkedVariable {name=Particles.Position, type=NiagaraPosition}}, B=<0 0 60>)
    Radius:NiagaraFloat=&Multiply_Float(
     A=300
     B=&FloatFromCurve(FloatCurve=DataInterface() {propertyValues {Curve {keys [<at 0.2155 0.2867>,<at 0.2875 0.9722>,<at 0.4239 0.9772>,<at 0.5419 0.8178>,<at 0.6689 1.0542>,<at 0.7012 1.0436>,<at 0.8454 0.7457>]}}})
    )
    Scale Color=1.25
   )
   /SetVariables(
    Particles.FlareStrike:Vector2f=&Multiply_Vector2(A=$Particles.FlashScale, B=<1.5 1.5>)
    Particles.LensFlare_Color:LinearColor=&MultiplyLinearColor(
     Linear Color=$Particles.Initial.LensFlare_Color
     Scale Factor=&MakeFloat4FromVectorAndFloat(
      "Vector (RGB)"=&VectorFromFloat(Value=&FloatFromCurve(FloatCurve=DataInterface() {propertyValues {Curve {keys [<at 0 1>,<at 0.1 1.5>,<at 0.3094 1.2707>,<at 1 1>]}}}))
      "Float (Alpha)"=&FloatFromCurve(FloatCurve=DataInterface() {propertyValues {Curve {keys [<at 0 0>,<at 0.1 1>,<at 0.75 1>,<at 1 0>]}}})
     )
    )
    Particles.LensFlare_Position:Vector3f=&Add_Vector(
     A=Linked() {linkedVariable {name=Particles.Position, type=NiagaraPosition}}
     B=&RotateVectorByQuaternion(
      Vector To Rotate=&MaskVectorBySpawnGroup(Masked Vector=$Particles.FlareOffset, Passthrough Vector=$Particles.FlareOffsetSecondary, Spawn Group Mask=$Particles.Initial.SpawnGroup)
      Quaternion=$Particles.MeshOrientation
     )
    )
   ) {isSetParameters=true, module=SetVariables_#7}
   &DynamicMaterialParameters(
    Index 0 Param 1=&FloatFromCurve(FloatCurve=DataInterface() {propertyValues {Curve {keys [
         <at 0 1 RCIM_Cubic>
         {arriveTangent=-1.2489, interpMode=RCIM_Cubic, leaveTangent=-1.2489, time=0.3515, value=0.6647}
         {arriveTangent=-1.8243, interpMode=RCIM_Cubic, leaveTangent=-1.8243, time=0.3565, value=0.5547}
         {arriveTangent=-2.8364, interpMode=RCIM_Cubic, leaveTangent=-2.8364, tangentMode=RCTM_User, time=0.6784, value=0.0685}
        ]}}})
    Index 0 Param 2=&FloatFromCurve(FloatCurve=DataInterface() {propertyValues {Curve {keys [
         {arriveTangent=3.0384, interpMode=RCIM_Cubic, leaveTangent=3.0384, tangentMode=RCTM_User, time=0, value=0.25}
         {arriveTangent=0.625, interpMode=RCIM_Cubic, leaveTangent=0.625, time=0.5, value=0.5}
         {arriveTangent=-1.25, interpMode=RCIM_Cubic, leaveTangent=-1.25, time=0.8, value=0.75}
         <at 0.9 0>
        ]}}})
    Index 0 Param 3=&FloatFromCurve(FloatCurve=DataInterface() {propertyValues {Curve {keys [
         {arriveTangent=0.8047, interpMode=RCIM_Cubic, leaveTangent=0.8047, tangentMode=RCTM_User, time=0, value=0.25}
         {arriveTangent=1.25, interpMode=RCIM_Cubic, leaveTangent=1.25, time=0.55, value=0.5}
         {arriveTangent=-2, interpMode=RCIM_Cubic, leaveTangent=-2, time=0.6, value=1}
         <at 0.8 0>
        ]}}})
    Module.Param0WriteEnabled=true
    Module.Param0WriteEnabled1=true
    Module.Param1WriteEnabled=true
    Module.Param1WriteEnabled1=true
    Module.Param2WriteEnabled=true
    Module.Param2WriteEnabled1=true
    Module.Param3WriteEnabled=false
    Module.Param3WriteEnabled1=true
    Write Parameter Index 1=true
   ) {module=DynamicMaterialParameters001}
  ]
  rendererClasses [
   &NiagaraMeshRendererProperties
   &NiagaraLightRendererProperties
   &NiagaraSpriteRendererProperties
  ]
  renderers [
   {rendererClass=&NiagaraMeshRendererProperties, rendererIndex=0}
   {rendererClass=&NiagaraLightRendererProperties, rendererIndex=1}
   {rendererClass=&NiagaraSpriteRendererProperties, rendererIndex=2}
   {rendererClass=&NiagaraSpriteRendererProperties, rendererIndex=3}
  ]
 }
 Initial_Lightning_Strike_RT {
  emitterUpdateScript [
   &EmitterState(
    Life Cycle Mode=::NewEnumerator1
    Loop Behavior=::NewEnumerator1
    Loop Delay=&UniformRangedFloat(Minimum=0, Maximum=0.2)
    Loop Duration=1
    UseLoopDelay=true
   )
   &SpawnBurst_Instantaneous(Spawn Count=1, Spawn Time=0.4, Use Spawn Probability=false)
   &SpawnBurst_Instantaneous(Spawn Count=1, Spawn Time=0.65, Use Spawn Probability=false) {module=SpawnBurst_Instantaneous001}
  ]
  particleSpawnScript [
   &FN_SetSpawnGroup_v2(Number of Spawn Groups=1)
   &FN_EvenNumberToSign(Input:NiagaraInt32=$Particles.SpawnGroup, Invert Sign=true, Output Float Attribute=true)
   &InitializeParticle(
    Lifetime=&UniformRangedFloat(Minimum=0.075, Maximum=0.1)
    Mesh Scale=&Multiply_Vector(
     A=&UniformRangedVector(Minimum=<1.25 0.8 0.8>, Maximum=<1.5 1.5 1.5>)
     B=&MakeVector(X=$Particles.Sign, Y=1, Z=1)
    )
    Module.Write Color=true
    Module.Write Lifetime=true
    Module.Write Mass=true
    Module.Write Material Random=true
    Module.Write Position=false
    Module.Write Scale=true
    Module.Write SpriteRotation=true
    Module.Write SpriteSize=true
    Position=&SimulationPosition()
    Sprite Rotation=&UniformRangedFloat(Minimum=0, Maximum=360)
    Sprite Size=&Vector2DFromFloat(Value=&UniformRangedFloat(Minimum=140, Maximum=150))
   )
   &DynamicMaterialParameters(
    Index 0 Param 4=&UniformRangedFloat(Minimum=0, Maximum=1)
    Index 2 Param 1=0.2
    Index 2 Param 2=1
    Module.Param0WriteEnabled=false
    Module.Param0WriteEnabled1=true
    Module.Param0WriteEnabled2=true
    Module.Param0WriteEnabled3=true
    Module.Param1WriteEnabled=false
    Module.Param1WriteEnabled1=true
    Module.Param1WriteEnabled2=true
    Module.Param1WriteEnabled3=true
    Module.Param2WriteEnabled=false
    Module.Param2WriteEnabled1=true
    Module.Param2WriteEnabled2=true
    Module.Param2WriteEnabled3=true
    Module.Param3WriteEnabled=true
    Module.Param3WriteEnabled1=true
    Module.Param3WriteEnabled2=true
    Module.Param3WriteEnabled3=true
    Write Parameter Index 1=true
    Write Parameter Index 2=true
    Write Parameter Index 3=true
   )
   &InitialMeshOrientation(
    Orientation Vector=<-1 0 0>
    Rotation=&UniformRangedVector(Minimum=<-0.5 0 0>, Maximum=<0.5 0 0>)
    Use Orientation Vector=true
    Use Rotation Vector=true
   )
   /SetVariables(Particles.FlashCameraOffset=10) {isSetParameters=true, module=SetVariables_#2}
   /SetVariables(
    Particles.FlareOffset:Vector3f=&UniformRangedVector(Minimum=<5 -1 -1>, Maximum=<100 1 1>)
    Particles.FlareSize:Vector2f=&Vector2DFromFloat(Value=&UniformRangedFloat(Minimum=55, Maximum=60))
    Particles.LensFlare_Color:LinearColor=&MultiplyLinearColor(
     Linear Color=$Particles.Color
     Scale Factor=&MakeVector4_FromLinearColor(LinearColor=$System.RandomFlareColor)
    )
   ) {isSetParameters=true, module=SetVariables_#3}
  ]
  particleUpdateScript [
   &SkeletalMeshLocation(
    Apply Sampled Orientation to Mesh Particles=true
    Bone / Socket Index=&Add_Integer(A=&SpawnGroup(), B=10)
    Bone Sampling Mode=::NewEnumerator4
    Inherit Velocity=false
    MaskSpawnGroup=false
    Skeletal Mesh=$User.SkeletalMesh
   )
   /SetVariables(Particles.MeshOrientation=&MultiplyQuaternion(Quaternion A=$Output.SkeletalMeshLocation.SampledRotation, Quaternion B=$Particles.Initial.MeshOrientation)) {isSetParameters=true, module=SetVariables_#4}
   /SetVariables(Particles.FlashPosition:Vector3f=&Add_Vector(A=Linked() {linkedVariable {name=Particles.Position, type=NiagaraPosition}}, B=<0 0 10>)) {isSetParameters=true, module=SetVariables_#5}
   /SetVariables(Particles.FlashColor:LinearColor=&ColorFromCurve(ColorCurve=DataInterface() {propertyValues {
       AlphaCurve {keys [<at 0.3517 0>,<at 0.3988 1>,<at 0.9585 1 RCIM_Cubic>,<at 1 0 RCIM_Cubic>]}
       BlueCurve {keys [<at 0.664 4 RCIM_Cubic>,<at 0.946 2 RCIM_Cubic>]}
       GreenCurve {keys [<at 0.664 3.6548 RCIM_Cubic>,<at 0.946 1.8793 RCIM_Cubic>]}
       RedCurve {keys [<at 0.664 2.5833 RCIM_Cubic>,<at 0.946 1.5048 RCIM_Cubic>]}
      }})) {isSetParameters=true, module=SetVariables_#6}
   &ParticleState()
   &ScaleColor(
    Scale RGB=&MakeVectorFromLinearColorRGB(Color=&ColorFromCurve(ColorCurve=DataInterface() {propertyValues {
        AlphaCurve {keys []}
        BlueCurve {keys [<at 0 2 RCIM_Cubic>,<at 1 1 RCIM_Cubic>]}
        GreenCurve {keys [<at 0 1.9431 RCIM_Cubic>,<at 1 0.8596 RCIM_Cubic>]}
        RedCurve {keys [<at 0 1.8021 RCIM_Cubic>,<at 1 0.6667 RCIM_Cubic>]}
       }}))
    ScaleA=true
    ScaleRGB=true
   )
   &FN_Light_v3(
    Color=&ColorFromCurve(ColorCurve=DataInterface() {propertyValues {
       AlphaCurve {keys []}
       BlueCurve {keys [
         <at 0.3882 4 RCIM_Cubic>
         <at 0.5 25 RCIM_Cubic>
         {arriveTangent=-18.6776, interpMode=RCIM_Cubic, leaveTangent=-18.6776, time=0.685, value=4}
         <at 0.8 19.3967 RCIM_Cubic>
         <at 1 4 RCIM_Cubic>
        ]}
       GreenCurve {keys [
         <at 0.3882 3.6948 RCIM_Cubic>
         <at 0.5 22.345 RCIM_Cubic>
         {arriveTangent=-12.3711, interpMode=RCIM_Cubic, leaveTangent=-12.3711, time=0.685, value=3.6948}
         <at 0.8 18.6337 RCIM_Cubic>
         <at 1 3.6948 RCIM_Cubic>
        ]}
       RedCurve {keys [
         <at 0.3882 3.2 RCIM_Cubic>
         <at 0.5 18.041 RCIM_Cubic>
         {arriveTangent=-2.1474, interpMode=RCIM_Cubic, leaveTangent=-2.1474, time=0.685, value=3.2}
         <at 0.8 17.3967 RCIM_Cubic>
         <at 1 3.2 RCIM_Cubic>
        ]}
      }})
    Create Color Attribute=true
    Create Exponent Attribute=true
    Create Position Attribute=true
    Create Radius Attribute=true
    Exponent=1.5
    Position:Vector3f=&Add_Vector(A=Linked() {linkedVariable {name=Particles.Position, type=NiagaraPosition}}, B=<0 0 60>)
    Radius:NiagaraFloat=&Multiply_Float(
     A=300
     B=&FloatFromCurve(FloatCurve=DataInterface() {propertyValues {Curve {keys [<at 0.2155 0.2867>,<at 0.2875 0.9722>,<at 0.4239 0.9772>,<at 0.5419 0.8178>,<at 0.6689 1.0542>,<at 0.7012 1.0436>,<at 0.8454 0.7457>]}}})
    )
    Scale Color=1.25
   )
   /SetVariables(
    Particles.FlareStrike:Vector2f=&Multiply_Vector2(A=$Particles.FlashScale, B=<1.5 1.5>)
    Particles.LensFlare_Color:LinearColor=&MultiplyLinearColor(
     Linear Color=$Particles.Initial.LensFlare_Color
     Scale Factor=&MakeFloat4FromVectorAndFloat(
      "Vector (RGB)"=&VectorFromFloat(Value=&FloatFromCurve(FloatCurve=DataInterface() {propertyValues {Curve {keys [<at 0 1>,<at 0.1 1.5>,<at 0.3094 1.2707>,<at 1 1>]}}}))
      "Float (Alpha)"=&FloatFromCurve(FloatCurve=DataInterface() {propertyValues {Curve {keys [<at 0 0>,<at 0.1 1>,<at 0.75 1>,<at 1 0>]}}})
     )
    )
    Particles.LensFlare_Position:Vector3f=&Add_Vector(
     A=Linked() {linkedVariable {name=Particles.Position, type=NiagaraPosition}}
     B=&RotateVectorByQuaternion(
      Vector To Rotate=&MaskVectorBySpawnGroup(Masked Vector=$Particles.FlareOffset, Passthrough Vector=$Particles.FlareOffsetSecondary, Spawn Group Mask=$Particles.Initial.SpawnGroup)
      Quaternion=$Particles.MeshOrientation
     )
    )
   ) {isSetParameters=true, module=SetVariables_#7}
   &DynamicMaterialParameters(
    Index 0 Param 1=&FloatFromCurve(FloatCurve=DataInterface() {propertyValues {Curve {keys [
         <at 0 1 RCIM_Cubic>
         {arriveTangent=-1.2489, interpMode=RCIM_Cubic, leaveTangent=-1.2489, time=0.3515, value=0.6647}
         {arriveTangent=-1.8243, interpMode=RCIM_Cubic, leaveTangent=-1.8243, time=0.3565, value=0.5547}
         {arriveTangent=-2.8364, interpMode=RCIM_Cubic, leaveTangent=-2.8364, tangentMode=RCTM_User, time=0.6784, value=0.0685}
        ]}}})
    Index 0 Param 2=&FloatFromCurve(FloatCurve=DataInterface() {propertyValues {Curve {keys [
         {arriveTangent=3.0384, interpMode=RCIM_Cubic, leaveTangent=3.0384, tangentMode=RCTM_User, time=0, value=0.25}
         {arriveTangent=0.625, interpMode=RCIM_Cubic, leaveTangent=0.625, time=0.5, value=0.5}
         {arriveTangent=-1.25, interpMode=RCIM_Cubic, leaveTangent=-1.25, time=0.8, value=0.75}
         <at 0.9 0>
        ]}}})
    Index 0 Param 3=&FloatFromCurve(FloatCurve=DataInterface() {propertyValues {Curve {keys [
         {arriveTangent=0.8047, interpMode=RCIM_Cubic, leaveTangent=0.8047, tangentMode=RCTM_User, time=0, value=0.25}
         {arriveTangent=1.25, interpMode=RCIM_Cubic, leaveTangent=1.25, time=0.55, value=0.5}
         {arriveTangent=-2, interpMode=RCIM_Cubic, leaveTangent=-2, time=0.6, value=1}
         <at 0.8 0>
        ]}}})
    Module.Param0WriteEnabled=true
    Module.Param0WriteEnabled1=true
    Module.Param1WriteEnabled=true
    Module.Param1WriteEnabled1=true
    Module.Param2WriteEnabled=true
    Module.Param2WriteEnabled1=true
    Module.Param3WriteEnabled=false
    Module.Param3WriteEnabled1=true
    Write Parameter Index 1=true
   ) {module=DynamicMaterialParameters001}
  ]
  rendererClasses [
   &NiagaraMeshRendererProperties
   &NiagaraLightRendererProperties
   &NiagaraSpriteRendererProperties
  ]
  renderers [
   {rendererClass=&NiagaraMeshRendererProperties, rendererIndex=0}
   {rendererClass=&NiagaraLightRendererProperties, rendererIndex=1}
   {rendererClass=&NiagaraSpriteRendererProperties, rendererIndex=2}
   {rendererClass=&NiagaraSpriteRendererProperties, rendererIndex=3}
  ]
 }
 Lightning_Arcs {
  emitterUpdateScript [
   &EmitterState(
    Delay First Loop Only=true
    Inactive Response=::NewEnumerator1
    Life Cycle Mode=::NewEnumerator1
    Loop Delay=2
    Loop Duration=&UniformRangedFloat(Minimum=0.25, Maximum=0.35)
    Recalculate Duration Each Loop=true
    UseLoopDelay=true
   )
   &FN_PartitionAttributes_v2(Particles Per Partition=3, Partitions=2)
   &SpawnBurst_Instantaneous(
    Spawn Count=&UniformRangedInt(Minimum=1, Maximum=2)
    Spawn Probability=&FloatFromCurve(
     FloatCurve=DataInterface() {propertyValues {Curve {keys [<at 0.7 1 RCIM_Constant>, {arriveTangent=-4.5166, interpMode=RCIM_Constant, leaveTangent=-4.5166, time=0.705, value=0}]}}}
     CurveIndex=$System.NormalizedLoopAge
    )
    Use Spawn Probability=true
   )
  ]
  particleSpawnScript [
   &FN_ParticlePartition_v2()
   &InitializeParticle(
    Color=<rgba 1.5312 1.8653 2 1>
    Lifetime=&UniformRangedFloat(Minimum=0.125, Maximum=0.175)
    Mesh Scale=&UniformRangedVector(Minimum=<6.5 6 6>, Maximum=<6.5 10 10>)
    Module.Write Color=true
    Module.Write Lifetime=true
    Module.Write Mass=true
    Module.Write Material Random=true
    Module.Write Position=false
    Module.Write Scale=true
    Module.Write SpriteRotation=false
    Module.Write SpriteSize=false
    Position=&SimulationPosition()
   )
   &FN_Initialize_BoneFlow_Attributes_v2(
    Angle:NiagaraFloat=&UniformRangedFloat(Minimum=0, Maximum=5)
    Radius=1.5
    T:NiagaraFloat=&UniformRangedFloat(Minimum=0, Maximum=1)
   )
   &DynamicMaterialParameters(
    Index 0 Param 1=6
    Index 0 Param 3=&UniformRangedFloat(Minimum=1.45, Maximum=2.2)
    Index 0 Param 4=&UniformRangedFloat(Minimum=-60, Maximum=60)
    Module.Param0WriteEnabled=true
    Module.Param1WriteEnabled=false
    Module.Param2WriteEnabled=true
    Module.Param3WriteEnabled=true
   )
   &InitialMeshOrientation(
    Rotation=&UniformRangedVector(Minimum=<-1 0 0>, Maximum=<1 0 0>)
    Use Orientation Vector=true
    Use Rotation Vector=true
   )
   /SetVariables(
    Particles.FlareOffset:Vector3f=&UniformRangedVector(Minimum=<0 0 -0.5>, Maximum=<8 0 0.5>)
    Particles.FlareSize:Vector2f=&Vector2DFromFloat(Value=&UniformRangedFloat(Minimum=10, Maximum=15))
    Particles.LensFlare_Color:LinearColor=&MultiplyLinearColor(
     Linear Color=$Particles.Color
     Scale Factor=&MakeVector4_FromLinearColor(LinearColor=$System.RandomFlareColor)
    )
   ) {isSetParameters=true, module=SetVariables_#3}
  ]
  particleUpdateScript [
   &ParticleState()
   &FN_BoneFlow_v2(
    Angle:NiagaraFloat=$Particles.FN_Initialize_BoneFlow_Attributes_v2.Angle
    Radius:NiagaraFloat=$Particles.FN_Initialize_BoneFlow_Attributes_v2.Radius
    SkeletalMesh:NiagaraDataInterfaceSkeletalMesh=$User.SkeletalMesh
    T:NiagaraFloat=&Add_Float(A=$Particles.FN_Initialize_BoneFlow_Attributes_v2.T, B=&Multiply_Float(A=$Particles.NormalizedAge, B=0.075))
   )
   /SetVariables(Particles.MeshOrientation=&MultiplyQuaternion(
     Quaternion A=&FindQuatBetweenTwoVectors(
      From Vector=<1 0 0>
      To Vector=&Lerp_Vector(
       A=<0 0 -1>
       B=&Multiply_VectorByFloat(Vector=$Output.FN_BoneFlow_v2.SplineTangent, Float=-1)
       Alpha=/Game/Effects/Niagara/DynamicInputs/Math/FN_InterpFloat(
        Interp type:ENiagaraInterpTypes=::NewEnumerator3
        A:NiagaraFloat=0.35
        B:NiagaraFloat=0.1
        Exponent:NiagaraInt32=2
        Value:NiagaraFloat=&Add_Float(A=$Particles.FN_Initialize_BoneFlow_Attributes_v2.T, B=&Multiply_Float(A=$Particles.NormalizedAge, B=-0.15))
       )
      )
     )
     Quaternion B=$Particles.Initial.MeshOrientation
    )) {isSetParameters=true, module=SetVariables_#4}
   &DynamicMaterialParameters(
    Index 0 Param 2=&FloatFromCurve(FloatCurve=DataInterface() {propertyValues {Curve {keys [{arriveTangent=-2.9283, interpMode=RCIM_Cubic, leaveTangent=-2.9283, tangentMode=RCTM_User, time=0.1, value=1}, <at 1 -1>]}}})
    Module.Param0WriteEnabled=false
    Module.Param1WriteEnabled=true
    Module.Param2WriteEnabled=false
    Module.Param3WriteEnabled=false
   ) {module=DynamicMaterialParameters001}
   &ScaleColor(
    Scale RGB=&MakeVectorFromLinearColorRGB(Color=&ColorFromCurve(ColorCurve=DataInterface() {propertyValues {
        AlphaCurve {keys []}
        BlueCurve {keys [<at 0 2 RCIM_Cubic>,<at 1 4 RCIM_Cubic>]}
        GreenCurve {keys [<at 0 1.8653 RCIM_Cubic>,<at 1 2.6922 RCIM_Cubic>]}
        RedCurve {keys [<at 0 1.5312 RCIM_Cubic>,<at 1 0.8958 RCIM_Cubic>]}
       }}))
    ScaleA=true
    ScaleRGB=true
   )
   &FN_Light_v3(
    Color=&ColorFromCurve(ColorCurve=DataInterface() {propertyValues {
       AlphaCurve {keys [<at 0 1 RCIM_Cubic>]}
       BlueCurve {keys [<at 0 4>, {arriveTangent=-2, interpMode=RCIM_Cubic, leaveTangent=-2, time=0.25, value=10}, <at 1 2 RCIM_Cubic>]}
       GreenCurve {keys [<at 0 3.5708>, {arriveTangent=-1.9086, interpMode=RCIM_Cubic, leaveTangent=-1.9086, time=0.25, value=8.8276}, <at 1 1.6622 RCIM_Cubic>]}
       RedCurve {keys [<at 0 2.875>, {arriveTangent=-1.7604, interpMode=RCIM_Cubic, leaveTangent=-1.7604, time=0.25, value=6.9271}, <at 1 1.1146 RCIM_Cubic>]}
      }})
    Create Color Attribute=true
    Create Exponent Attribute=true
    Create Position Attribute=true
    Create Radius Attribute=true
    Exponent=4
    Position:Vector3f=&Add_Vector(
     A=Linked() {linkedVariable {name=Particles.Position, type=NiagaraPosition}}
     B=&RotateVectorByQuaternion(Vector To Rotate=<60 0 0>, Quaternion=$Particles.MeshOrientation)
    )
    Radius:NiagaraFloat=&Multiply_Float(
     A=225
     B=&FloatFromCurve(
      FloatCurve=DataInterface() {propertyValues {Curve {keys [<at 0.79 1 RCIM_Cubic>, {arriveTangent=-4.5166, interpMode=RCIM_Cubic, leaveTangent=-4.5166, tangentMode=RCTM_User, time=0.8, value=0}]}}}
      CurveIndex=$System.NormalizedLoopAge
     )
    )
    Scale Color=1.25
   )
   /SetVariables(
    Particles.FlareStrike:Vector2f=&Multiply_Vector2(A=$Particles.FlashScale, B=<1.5 1.5>)
    Particles.LensFlare_Color:LinearColor=&MultiplyLinearColor(
     Linear Color=$Particles.Initial.LensFlare_Color
     Scale Factor=&MakeFloat4FromVectorAndFloat(
      "Vector (RGB)"=&VectorFromFloat(Value=&FloatFromCurve(FloatCurve=DataInterface() {propertyValues {Curve {keys [<at 0 1>,<at 0.1 1>,<at 0.3094 0.6>,<at 1 1>]}}}))
      "Float (Alpha)"=&FloatFromCurve(FloatCurve=DataInterface() {propertyValues {Curve {keys [<at 0 0>,<at 0.1 0.175>,<at 0.75 0.175>,<at 1 0>]}}})
     )
    )
    Particles.LensFlare_Position:Vector3f=&Add_Vector(
     A=Linked() {linkedVariable {name=Particles.Position, type=NiagaraPosition}}
     B=&RotateVectorByQuaternion(
      Vector To Rotate=&MaskVectorBySpawnGroup(Masked Vector=$Particles.FlareOffset, Passthrough Vector=$Particles.FlareOffsetSecondary, Spawn Group Mask=$Particles.Initial.SpawnGroup)
      Quaternion=$Particles.MeshOrientation
     )
    )
   ) {isSetParameters=true, module=SetVariables_#7}
  ]
  rendererClasses [
   &NiagaraMeshRendererProperties
   &NiagaraLightRendererProperties
   &NiagaraSpriteRendererProperties
  ]
  renderers [
   {rendererClass=&NiagaraMeshRendererProperties, rendererIndex=0}
   {rendererClass=&NiagaraLightRendererProperties, rendererIndex=1}
   {rendererClass=&NiagaraSpriteRendererProperties, rendererIndex=2}
  ]
 }
 Rotating_Air_Meshes {
  emitterUpdateScript [
   &EmitterState(Delay First Loop Only=true, Life Cycle Mode=::NewEnumerator1, Loop Delay=1, UseLoopDelay=true)
   &SpawnRate(SpawnRate=10, Use Spawn Probability=false)
  ]
  particleSpawnScript [
   &InitializeParticle(
    Color=<rgba 0.0543 0.0795 0.1198 0.65>
    Lifetime=&UniformRangedFloat(Minimum=1.5, Maximum=2)
    Mesh Scale=&UniformRangedVector(Minimum=<4 4 9>, Maximum=<6 6 15>)
    Module.Write Color=true
    Module.Write Lifetime=true
    Module.Write Mass=true
    Module.Write Material Random=true
    Module.Write Position=false
    Module.Write Scale=true
    Module.Write SpriteRotation=false
    Module.Write SpriteSize=false
    Position=&SimulationPosition()
   )
   &InitialMeshOrientation(
    Rotation=&UniformRangedVector(Minimum=<0 0 -1>, Maximum=<0 0 1>)
    Use Orientation Vector=true
    Use Rotation Vector=true
   )
   &SystemLocation(
    MaskSpawnGroup=false
    Offset=&UniformRangedVector(Minimum=<0 -10 0>, Maximum=<0 -10 25>)
    Offset Coordinate Space=::Simulation
   )
   &AddVelocity(Velocity=&UniformRangedVector(Minimum=<0 0 50>, Maximum=<0 0 65>))
   &DynamicMaterialParameters(
    Index 0 Param 1=&UniformRangedFloat(Minimum=0.1, Maximum=0.25)
    Index 0 Param 2=1
    Module.Param0WriteEnabled=true
    Module.Param1WriteEnabled=true
    Module.Param2WriteEnabled=true
    Module.Param3WriteEnabled=true
   )
   &FN_MeshRenderer_v2(
    Color:LinearColor=&UniformRangedLinearColor(Minimum=<rgba 0.8802 0.8802 0.8802 0.15>, Maximum=<rgba 0.8802 0.8802 0.8802 0.5>)
    Create COlor Attribute=true
    Create Dynamic Param Attribute=true
    Create Orientation Attribute=true
    Create Scale Attribute=true
    DynamicMaterial:Vector4f=&MakeVector4(
     X=&UniformRangedFloat(Minimum=0, Maximum=1)
     Y=&UniformRangedFloat(Minimum=5, Maximum=6)
     Z=0
     W=2
    )
    Scale:Vector3f=&UniformRangedVector(Minimum=<0.6 0.6 1.15>, Maximum=<0.8 0.8 1.65>)
   )
  ]
  particleUpdateScript [
   &ScaleVelocity(Velocity Scale=&VectorFromFloat(Value=&FloatFromCurve(
      FloatCurve=DataInterface() {propertyValues {Curve {keys [<at 0.6 1 RCIM_Cubic>, {arriveTangent=-4.5166, interpMode=RCIM_Cubic, leaveTangent=-4.5166, tangentMode=RCTM_User, time=0.675, value=-0.5}]}}}
      CurveIndex=$System.NormalizedLoopAge
     )))
   &SolveForcesAndVelocity(Clamp Velocity=false, Limit Acceleration=false)
   &FN_MeshRenderer_v2(
    Color:LinearColor=&MultiplyLinearColor(
     Linear Color=$Particles.Initial.FN_MeshRenderer_v2.Color
     Scale Factor=&MakeFloat4FromVectorAndFloat(
      "Vector (RGB)"=<1 1 1>
      "Float (Alpha)"=&Multiply_Float(
       A=&FloatFromCurve(FloatCurve=DataInterface() {propertyValues {Curve {keys [
            <at 0 0 RCIM_Cubic>
            {arriveTangent=1.3333, leaveTangent=1.3333, time=0.15, value=1}
            <at 0.85 1 RCIM_Cubic>
            {arriveTangent=-4.5166, interpMode=RCIM_Cubic, leaveTangent=-4.5166, tangentMode=RCTM_User, time=1, value=0}
           ]}}})
       B=&FloatFromCurve(
        FloatCurve=DataInterface() {propertyValues {Curve {keys [<at 0.65 1 RCIM_Cubic>, {arriveTangent=-4.5166, interpMode=RCIM_Cubic, leaveTangent=-4.5166, tangentMode=RCTM_User, time=0.8, value=0}]}}}
        CurveIndex=$System.NormalizedLoopAge
       )
      )
     )
    )
    Create COlor Attribute=true
    Create Scale Attribute=true
    Scale:Vector3f=&Multiply_Vector(
     A=$Particles.Initial.FN_MeshRenderer_v2.Scale
     B=&VectorFromCurve(VectorCurve=DataInterface() {propertyValues {
        XCurve {keys [<at 0 1.25 RCIM_Cubic>,<at 1 0.5>]}
        YCurve {keys [<at 0 1.25 RCIM_Cubic>,<at 1 0.5>]}
        ZCurve {keys [<at 0 0.6 RCIM_Cubic>,<at 1 1.25>]}
       }})
    )
   ) {module=FN_MeshRenderer_v2001}
   &ParticleState()
   &UpdateMeshOrientation(Rotation Rate:NiagaraFloat=&UniformRangedFloat(Minimum=1.5, Maximum=2), Rotation Vector:Vector3f=<0 0 1>)
   &ScaleMeshSize(Scale Factor=&VectorFromCurve(VectorCurve=DataInterface() {propertyValues {XCurve {keys [<at 0 1.25>,<at 1 0.45>]}, YCurve {keys [<at 0 1.25>,<at 1 0.45>]}, ZCurve {keys [<at 0 0.6>,<at 1 1.2>]}}}))
   &ScaleColor(
    Scale Alpha=&Multiply_Float(
     A=&FloatFromCurve(FloatCurve=DataInterface() {propertyValues {Curve {keys [
          <at 0 0 RCIM_Cubic>
          <at 0.3 1 RCIM_Cubic>
          {arriveTangent=-4.5166, interpMode=RCIM_Cubic, leaveTangent=-4.5166, tangentMode=RCTM_User, time=1, value=0}
         ]}}})
     B=&FloatFromCurve(
      FloatCurve=DataInterface() {propertyValues {Curve {keys [<at 0.65 1 RCIM_Cubic>, {arriveTangent=-4.5166, interpMode=RCIM_Cubic, leaveTangent=-4.5166, tangentMode=RCTM_User, time=0.8, value=0}]}}}
      CurveIndex=$System.NormalizedLoopAge
     )
    )
    ScaleA=true
    ScaleRGB=false
   )
  ]
  rendererClasses [&NiagaraMeshRendererProperties]
  renderers [
   {rendererClass=&NiagaraMeshRendererProperties, rendererIndex=0}
   {rendererClass=&NiagaraMeshRendererProperties, rendererIndex=1}
  ]
 }
 Smoke_Vortex {
  emitterUpdateScript [
   &EmitterState(Delay First Loop Only=true, Life Cycle Mode=::NewEnumerator1, Loop Delay=1, Loop Duration=1, UseLoopDelay=true)
   &SpawnRate(SpawnRate=34, Use Spawn Probability=false)
  ]
  particleSpawnScript [
   &InitializeParticle(
    Color=&UniformRangedLinearColor(Minimum=<rgba 0.043 0.0547 0.0677 0.35>, Maximum=<rgba 0.0354 0.0557 0.0781 0.45>)
    Lifetime=&UniformRangedFloat(Minimum=0.25, Maximum=0.5)
    Module.Write Color=true
    Module.Write Lifetime=true
    Module.Write Mass=true
    Module.Write Material Random=true
    Module.Write Position=false
    Module.Write Scale=false
    Module.Write SpriteRotation=true
    Module.Write SpriteSize=true
    Position=&SimulationPosition()
    Sprite Rotation=&NormalizedAngleToDegrees(NormalizedAngle=&UniformRangedFloat(Minimum=-1, Maximum=1))
    Sprite Size=&Vector2DFromFloat(Value=&UniformRangedFloat(Minimum=150, Maximum=200))
   )
   /Niagara/Modules/Spawn/Location/SphereLocation(Hemisphere Z=true, MaskSpawnGroup=false, Module.Override Randomness=false, Sphere Radius=10, Surface Only=true)
   &AddVelocity(Velocity=&Multiply_VectorByFloat(Vector=$Output.SphereLocation.SphereNormal, Float=500))
  ]
  particleUpdateScript [
   &SubUVAnimation(Number Of Frames=64, SubUV Lookup Index=&Multiply_Float(A=$Particles.NormalizedAge, B=0.75))
   &ParticleState()
   &Drag(Drag=2, Use Linear Drag=true, Use Rotational Drag=false)
   &SolveForcesAndVelocity(Clamp Velocity=false, Limit Acceleration=false)
   &SpriteRotationRate(Rotation Rate=50)
   &ScaleColor(
    Scale Alpha=&Multiply_Float(
     A=&FloatFromCurve(FloatCurve=DataInterface() {propertyValues {Curve {keys [<at 0 0 RCIM_Cubic>, {arriveTangent=-0.1, interpMode=RCIM_Cubic, leaveTangent=-0.1, time=0.1, value=1}, <at 1 -0.1>]}}})
     B=&FloatFromCurve(
      FloatCurve=DataInterface() {propertyValues {Curve {keys [<at 0.65 1 RCIM_Cubic>, {arriveTangent=-4.5166, interpMode=RCIM_Cubic, leaveTangent=-4.5166, tangentMode=RCTM_User, time=0.8, value=0}]}}}
      CurveIndex=$System.NormalizedLoopAge
     )
    )
    ScaleA=true
    ScaleRGB=false
   )
  ]
  rendererClasses [&NiagaraSpriteRendererProperties]
  renderers [{rendererClass=&NiagaraSpriteRendererProperties}]
 }
 Smokes {
  emitterUpdateScript [
   &EmitterState(Life Cycle Mode=::NewEnumerator1, Loop Behavior=::NewEnumerator1, Loop Delay=1, Loop Duration=2, UseLoopDelay=true)
   &SpawnBurst_Instantaneous(Spawn Count=16, Use Spawn Probability=false)
   &SpawnRate(SpawnRate=&UniformRangedFloat(Minimum=10, Maximum=20), Use Spawn Probability=false)
  ]
  particleSpawnScript [
   &InitializeParticle(
    Color=<rgba 0.5677 0.5677 0.5677 0.1>
    Lifetime=&UniformRangedFloat(Minimum=1, Maximum=1)
    Module.Write Color=true
    Module.Write Lifetime=true
    Module.Write Mass=true
    Module.Write Material Random=true
    Module.Write Position=false
    Module.Write Scale=false
    Module.Write SpriteRotation=true
    Module.Write SpriteSize=true
    Position=&SimulationPosition()
    Sprite Rotation=&UniformRangedFloat(Minimum=0, Maximum=360)
    Sprite Size=&Vector2DFromFloat(Value=&UniformRangedFloat(Minimum=30, Maximum=40))
   )
   &CylinderLocation(
    Coordinate Space:ENiagaraCoordinateSpace=::World
    Cylinder Height=0
    Cylinder Radius=25
    MaskSpawnGroup=false
    Module.Enable Lathe Profile=false
    Module.Override Randomness=false
    Module.Surface Only=true
    Offset=<0 0 50>
    Surface Only Band Thickness=1
    Use Endcaps In Surface Only Mode=false
   )
   &AddVelocity(Velocity=<0 0 -300>)
   &AddVelocityFromPoint(Module.Use Falloff Distance=false, Origin Offset=<0 0 50>, Velocity Strength=200)
   &ApplyInitialForces()
  ]
  particleUpdateScript [
   &ParticleState()
   &AccelerationForce(Acceleration=<0 0 500>)
   &Drag(Drag=2, Use Linear Drag=true, Use Rotational Drag=false)
   &SolveForcesAndVelocity(Clamp Velocity=false, Limit Acceleration=false)
   &ScaleSpriteSize(Scale Factor=&Vector2DFromFloat(Value=&FloatFromCurve(FloatCurve=DataInterface() {propertyValues {Curve {keys [
          {arriveTangent=1.491, interpMode=RCIM_Cubic, leaveTangent=1.491, tangentMode=RCTM_User, time=-0.0046, value=1.1969}
          <at 0.5 3.4809 RCIM_Cubic>
         ]}}})))
   /SetVariables(Particles.SubImageIndex=&FloatFromCurve(FloatCurve=DataInterface() {propertyValues {Curve {keys [
         {arriveTangent=29.8623, interpMode=RCIM_Cubic, leaveTangent=29.8623, tangentMode=RCTM_User, time=-0.0046, value=5.0556}
         <at 1.0011 22.1689 RCIM_Cubic>
        ]}}})) {isSetParameters=true}
   &DynamicMaterialParameters(
    Index 0 Param 1=&FloatFromCurve(FloatCurve=DataInterface() {propertyValues {Curve {keys [<at 0.3666 1.0113 RCIM_Cubic>,<at 1.0011 -0.8759 RCIM_Cubic>]}}})
    Index 0 Param 4=&FloatFromCurve(FloatCurve=DataInterface() {propertyValues {Curve {keys [<at -0.0046 0.2 RCIM_Cubic>,<at 1.0011 1 RCIM_Cubic>]}}})
    Module.Param0WriteEnabled=true
    Module.Param1WriteEnabled=true
    Module.Param2WriteEnabled=true
    Module.Param3WriteEnabled=true
   )
   &CameraOffset(Camera Offset Amount=10)
   &ScaleColor(
    Scale Alpha=&FloatFromCurve(FloatCurve=DataInterface() {propertyValues {Curve {keys [<at 0 0 RCIM_Cubic>,<at 0.2 1 RCIM_Cubic>,<at 1 0 RCIM_Cubic>]}}})
    ScaleA=true
    ScaleRGB=true
   )
  ]
  rendererClasses [&NiagaraSpriteRendererProperties]
  renderers [{rendererClass=&NiagaraSpriteRendererProperties}]
 }
}
summary {
 systemName=NS_Emote_HighTower_Squash
 emitters [
  {emitterName=BoneFlowElectricArcs, rendererClasses [&NiagaraSpriteRendererProperties]}
  {emitterName=Rotating_Air_Meshes, rendererClasses [&NiagaraMeshRendererProperties]}
  {bEnabled=false, emitterName=Air_Swooping_Up_Mesh, rendererClasses [&NiagaraMeshRendererProperties]}
  {emitterName=Smoke_Vortex, rendererClasses [&NiagaraSpriteRendererProperties]}
  {emitterName=Debris_Vortex, rendererClasses [&NiagaraSpriteRendererProperties]}
  {
   emitterName=Lightning_Arcs
   rendererClasses [
    &NiagaraMeshRendererProperties
    &NiagaraLightRendererProperties
    &NiagaraSpriteRendererProperties
   ]
  }
  {emitterName=BoneFlowElectricArcsCenterBody, rendererClasses [&NiagaraSpriteRendererProperties]}
  {
   emitterName=Initial_Lightning_Strike_LT
   rendererClasses [
    &NiagaraMeshRendererProperties
    &NiagaraLightRendererProperties
    &NiagaraSpriteRendererProperties
   ]
  }
  {
   emitterName=Dark_Spot_and_Upward_Vortex
   rendererClasses [&NiagaraSpriteRendererProperties, &NiagaraMeshRendererProperties]
  }
  {emitterName=Eye_Glows_and_Trails, rendererClasses [&NiagaraMeshRendererProperties]}
  {emitterName=Eye_Electric_Arcs, rendererClasses [&NiagaraSpriteRendererProperties]}
  {
   emitterName=Initial_Lightning_Strike_RT
   rendererClasses [
    &NiagaraMeshRendererProperties
    &NiagaraLightRendererProperties
    &NiagaraSpriteRendererProperties
   ]
  }
  {
   emitterName=FN_BoneFlow_v2
   rendererClasses [&NiagaraSpriteRendererProperties, &NiagaraLightRendererProperties]
  }
  {emitterName=FN_BoneFlow_v2001, rendererClasses [&NiagaraSpriteRendererProperties]}
  {emitterName=Smokes, rendererClasses [&NiagaraSpriteRendererProperties]}
  {emitterName=Decal, rendererClasses [&NiagaraSpriteRendererProperties]}
 ]
 userVariables [
  User.bIsFrontEndPreview {type=NiagaraBool}
  User.SkeletalMesh {
   type=NiagaraDataInterfaceSkeletalMesh
   defaultValue {
    struct=NiagaraExt_VariableValue_DataInterface
    value {dataInterface=&self:NiagaraDataInterfaceSkeletalMesh_0}
   }
  }
  User.SkeletalMesh2 {
   type=NiagaraDataInterfaceSkeletalMesh
   defaultValue {
    struct=NiagaraExt_VariableValue_DataInterface
    value {dataInterface=&self:NiagaraDataInterfaceSkeletalMesh_1}
   }
  }
  User.SkeletalMesh_03 {
   type=NiagaraDataInterfaceSkeletalMesh
   defaultValue {
    struct=NiagaraExt_VariableValue_DataInterface
    value {dataInterface=&self:NiagaraDataInterfaceSkeletalMesh_2}
   }
  }
 ]
}
